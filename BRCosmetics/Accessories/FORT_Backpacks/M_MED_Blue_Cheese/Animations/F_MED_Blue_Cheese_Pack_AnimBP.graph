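# Animation Blueprint for the Blue Cheese backpack cosmetic that plays idle animations and celebrates when an enemy player is eliminated.
# The Idle state randomly plays the idle or, less often, a break animation, while On Kill plays the celebrate clip and returns automatically.
# The EnemyPlayerEliminated anim input event sets the transition flag, which clears after each animation evaluation.
# A lid bone offset and a second random player in AnimGraph are unconnected to the output.
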AnimBlueprint /BRCosmetics/Accessories/FORT_Backpacks/M_MED_Blue_Cheese/Animations/F_MED_Blue_Cheese_Pack_AnimBP
target_skeleton=/BRCosmetics/Accessories/FORT_Backpacks/M_MED_Blue_Cheese/Meshes/M_MED_Blue_Cheese_Pack_Skeleton
graphs [
 AnimGraph {nodes [
   n1 = "Transform (Modify) Bone - Bone: dyn_lid_parent"(ComponentPose=Local To Component(), Alpha=1, Translation="-0.5,0,0", Rotation="0,0,0", Scale="1,1,1") {
    translationMode=BMM_Additive
    translationSpace=BCS_ParentBoneSpace
    alphaBoolBlend {blendOption=Linear}
    boneToModify {boneName=dyn_lid_parent}
    translation {x=-0.5}
   }
   Component To Local(ComponentPose=@n1)
   Local Space Ref Pose()
   Random Sequence Player!() -> (Pose:Pose Link Structure) {
    is=title AnimGraphNode_RandomPlayer
    entries [
     {
      chanceToPlay=10
      maxLoopCount=0
      maxPlayRate=1
      minLoopCount=0
      minPlayRate=1
      sequence=&/Blue_Cheese_Backpack_Idle
      blendIn {blendOption=Linear, blendTime=0.1, customCurve=null}
     }
     {
      chanceToPlay=1
      maxLoopCount=0
      maxPlayRate=1
      minLoopCount=0
      minPlayRate=1
      sequence=&/Blue_Cheese_Backpack_Idle_Break
      blendIn {blendOption=Linear, blendTime=0.075, customCurve=null}
     }
     {
      chanceToPlay=1
      maxLoopCount=0
      maxPlayRate=1
      minLoopCount=0
      minPlayRate=1
      sequence=&/Blue_Cheese_Backpack_Idle_Celebrate
      blendIn {blendOption=Linear, blendTime=0.1, customCurve=null}
     }
    ]
   }
   AnimGraphNode_StateMachine_0 = Idling StateMachine() {is=title AnimGraphNode_StateMachine}
   Output Pose(Result=@AnimGraphNode_StateMachine_0) {layerGroup=None}
  ]}
 Idling StateMachine {
  entry=Idle
  path=AnimGraph.AnimGraphNode_StateMachine_0.Idling StateMachine
  ids=id
  nodes [
   AnimStateNode_0 = Idle()
   AnimStateNode_3 = On Kill() {is=title AnimStateNode}
   AnimStateTransitionNode_1 = AnimStateTransitionNode()
   AnimStateTransitionNode_2 = AnimStateTransitionNode() {bAutomaticRuleBasedOnSequencePlayerInState=true}
  ]
  links [
   {
    dst=On Kill
    rule=AnimGraph.AnimGraphNode_StateMachine_0.Idling StateMachine.AnimStateTransitionNode_1.Transition
    src=Idle
    via=AnimStateTransitionNode_1
   }
   {
    dst=Idle
    rule=AnimGraph.AnimGraphNode_StateMachine_0.Idling StateMachine.AnimStateTransitionNode_2.Transition
    src=On Kill
    via=AnimStateTransitionNode_2
   }
  ]
 }
 Idle {
  path=AnimGraph.AnimGraphNode_StateMachine_0.Idling StateMachine.AnimStateNode_0.Idle
  nodes [
   n1 = Random Sequence Player() {entries [
     {
      chanceToPlay=0.75
      maxLoopCount=0
      maxPlayRate=1
      minLoopCount=0
      minPlayRate=1
      sequence=&Blue_Cheese_Backpack_Idle
      blendIn {blendOption=Linear, blendTime=0.2, customCurve=null}
     }
     {
      chanceToPlay=0.25
      maxLoopCount=0
      maxPlayRate=1
      minLoopCount=0
      minPlayRate=1
      sequence=&Blue_Cheese_Backpack_Idle_Break
      blendIn {blendOption=Linear, blendTime=0.2, customCurve=null}
     }
    ]}
   Output Animation Pose(Result=@n1)
  ]
 }
 On Kill {
  path=AnimGraph.AnimGraphNode_StateMachine_0.Idling StateMachine.AnimStateNode_3.On Kill
  nodes [Output Animation Pose(Result=Sequence Player 'Blue_Cheese_Backpack_Idle_Celebrate'() {sequence=&Blue_Cheese_Backpack_Idle_Celebrate})]
 }
 Transition {
  path=AnimGraph.AnimGraphNode_StateMachine_0.Idling StateMachine.AnimStateTransitionNode_1.Transition
  nodes [Result(bCanEnterTransition=Get Enemy Player Eliminated() -> (Enemy Player Eliminated:bool))]
 }
 Transition {path=AnimGraph.AnimGraphNode_StateMachine_0.Idling StateMachine.AnimStateTransitionNode_2.Transition, nodes [Result()]}
 EventGraph {nodes [
   n1 = Event BlueprintProcessAnimInputEvent()
   > Branch(Condition=IsNamed(self=@n1.AnimInputEvent, InEventName=aie.EnemyPlayerEliminated))
   > Set Enemy Player Eliminated(Enemy Player Eliminated:bool=true) -> (Output_Get:bool)
   Event BlueprintPostEvaluateAnimation() {is=title K2Node_Event}
   > Set Enemy Player Eliminated(Enemy Player Eliminated=false)
   TryGetPawnOwner()
   Event BlueprintUpdateAnimation()
  ]}
]
variables [Enemy Player Eliminated {type=boolean}]
